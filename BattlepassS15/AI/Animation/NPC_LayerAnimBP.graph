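# Animation layer Blueprint for BattlepassS15 NPCs on the Fortnite player skeleton, driving a full-body conversation pose override.
# Named anim input events set one-frame bools, cleared after evaluation, that trigger the NPC_Conversation state machine.
# It moves from additive-noise idle through start and loop to finish; give-item and receive-item states exit only on conversation end.
# The FullBodyOverride output scales both weapon bones to near zero, effectively hiding held weapons.
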AnimBlueprint /BattlepassS15/AI/Animation/NPC_LayerAnimBP
target_skeleton=&Fortnite_M_Avg_Player_Skeleton
graphs [
 AnimGraph {nodes [Output Pose() {layerGroup=None}]}
 CharacterSkeletalControlPostLegIK {nodes [
   Input Pose() {name=InPosePostLegIK}
   Output Pose() {layerGroup=PlayerOverride}
  ]}
 EventGraph {nodes [
   n1 = Event BlueprintProcessAnimInputEvent()
   > n2 = Branch(Condition=IsNamed(self=@n1.AnimInputEvent, InEventName=aie.npc.ConversationStart))
   > Set IsConversationStarting(IsConversationStarting:bool=true) -> (Output_Get:bool)
   n3 = Branch(execute=@n2.else, Condition=IsNamed(self=@n1.AnimInputEvent, InEventName=aie.npc.ConversationEnd))
   > Set IsConversationEnding(IsConversationEnding:bool=true) -> (Output_Get:bool)
   n4 = Branch(execute=@n3.else, Condition=IsNamed(self=@n1.AnimInputEvent, InEventName=aie.npc.GiveItem))
   > Set IsGivingItem(IsGivingItem:bool=true) -> (Output_Get:bool)
   Branch(execute=@n4.else, Condition=IsNamed(self=@n1.AnimInputEvent, InEventName=aie.npc.ReceiveItem))
   > Set IsReceivingItem(IsReceivingItem:bool=true) -> (Output_Get:bool)
   Event BlueprintPostEvaluateAnimation() {is=title K2Node_Event}
   > Set IsConversationStarting(IsConversationStarting=false)
   > Set IsConversationEnding(IsConversationEnding=false)
   > Set IsGivingItem(IsGivingItem=false)
   > Set IsReceivingItem(IsReceivingItem=false)
   TryGetPawnOwner()
   Event BlueprintUpdateAnimation()
  ]}
 FullBodyOverride {nodes [
   AnimGraphNode_StateMachine_0 = NPC_Conversation() {is=title AnimGraphNode_StateMachine}
   n1 = "Transform (Modify) Bone - Bone: weapon_l"!(ComponentPose=Local To Component(LocalPose=@AnimGraphNode_StateMachine_0), Alpha=1, Scale="0.001,0.001,0.001") -> (Pose) {scaleMode=BMM_Replace, alphaBoolBlend {blendOption=Linear}, boneToModify {boneName=weapon_l}, scale=<0.001 0.001 0.001>}
   n2 = "Transform (Modify) Bone - Bone: weapon_r"!(ComponentPose=@n1, Alpha=1, Scale="0.001,0.001,0.001") -> (Pose) {scaleMode=BMM_Replace, alphaBoolBlend {blendOption=Linear}, boneToModify {boneName=weapon_r}, scale=<0.001 0.001 0.001>}
   Output Pose(Result=Component To Local(ComponentPose=@n2)) {layerGroup=PlayerOverride}
   Save cached pose 'PassThroughFullBodyPose'(Pose=Input Pose() {name=PassThroughFullBodyPose}) {cacheName=PassThroughFullBodyPose}
  ]}
 NPC_Conversation {
  entry=Default
  path=FullBodyOverride.AnimGraphNode_StateMachine_0.NPC_Conversation
  ids=id
  nodes [
   AnimStateNode_0 = Default() {is=title AnimStateNode}
   AnimStateNode_1 = ConversationStart() {is=title AnimStateNode}
   AnimStateNode_2 = ConversationLoop() {is=title AnimStateNode}
   AnimStateNode_3 = ConversationFinish() {is=title AnimStateNode}
   AnimStateNode_4 = ConversationGiveItem() {is=title AnimStateNode}
   AnimStateNode_5 = ConversationReceiveItem() {is=title AnimStateNode}
   AnimStateTransitionNode_0 = AnimStateTransitionNode()
   AnimStateTransitionNode_1 = AnimStateTransitionNode() {bAutomaticRuleBasedOnSequencePlayerInState=true}
   AnimStateTransitionNode_2 = AnimStateTransitionNode()
   AnimStateTransitionNode_3 = AnimStateTransitionNode() {bAutomaticRuleBasedOnSequencePlayerInState=true}
   AnimStateTransitionNode_4 = AnimStateTransitionNode()
   AnimStateTransitionNode_5 = AnimStateTransitionNode()
   AnimStateTransitionNode_6 = AnimStateTransitionNode()
   AnimStateTransitionNode_7 = AnimStateTransitionNode()
  ]
  links [
   {
    dst=Default
    rule=FullBodyOverride.AnimGraphNode_StateMachine_0.NPC_Conversation.AnimStateTransitionNode_3.Transition
    src=ConversationFinish
    via=AnimStateTransitionNode_3
   }
   {
    dst=ConversationFinish
    rule=FullBodyOverride.AnimGraphNode_StateMachine_0.NPC_Conversation.AnimStateTransitionNode_6.Transition
    src=ConversationGiveItem
    via=AnimStateTransitionNode_6
   }
   {
    dst=ConversationFinish
    rule=FullBodyOverride.AnimGraphNode_StateMachine_0.NPC_Conversation.AnimStateTransitionNode_2.Transition
    src=ConversationLoop
    via=AnimStateTransitionNode_2
   }
   {
    dst=ConversationGiveItem
    rule=FullBodyOverride.AnimGraphNode_StateMachine_0.NPC_Conversation.AnimStateTransitionNode_4.Transition
    src=ConversationLoop
    via=AnimStateTransitionNode_4
   }
   {
    dst=ConversationReceiveItem
    rule=FullBodyOverride.AnimGraphNode_StateMachine_0.NPC_Conversation.AnimStateTransitionNode_5.Transition
    src=ConversationLoop
    via=AnimStateTransitionNode_5
   }
   {
    dst=ConversationFinish
    rule=FullBodyOverride.AnimGraphNode_StateMachine_0.NPC_Conversation.AnimStateTransitionNode_7.Transition
    src=ConversationReceiveItem
    via=AnimStateTransitionNode_7
   }
   {
    dst=ConversationLoop
    rule=FullBodyOverride.AnimGraphNode_StateMachine_0.NPC_Conversation.AnimStateTransitionNode_1.Transition
    src=ConversationStart
    via=AnimStateTransitionNode_1
   }
   {
    dst=ConversationStart
    rule=FullBodyOverride.AnimGraphNode_StateMachine_0.NPC_Conversation.AnimStateTransitionNode_0.Transition
    src=Default
    via=AnimStateTransitionNode_0
   }
  ]
 }
 Default {
  path=FullBodyOverride.AnimGraphNode_StateMachine_0.NPC_Conversation.AnimStateNode_0.Default
  nodes [Output Animation Pose(Result=Apply Additive(
     Base=Sequence Player 'ConversationNPCBasePose'() {sequence=&/ConversationNPCBasePose}
     Additive="Sequence Player (additive) 'Idle_Noise'"() {sequence=/Game/Animation/Game/MainPlayer/Locomotion/Medium/Male/Idle/Idle_Noise}
    ) {alphaBoolBlend {blendOption=Linear}})]
 }
 ConversationStart {
  path=FullBodyOverride.AnimGraphNode_StateMachine_0.NPC_Conversation.AnimStateNode_1.ConversationStart
  nodes [Output Animation Pose(Result=Sequence Player 'ConversationStart'() {bLoopAnimation=false, sequence=&/ConversationStart})]
 }
 ConversationLoop {
  path=FullBodyOverride.AnimGraphNode_StateMachine_0.NPC_Conversation.AnimStateNode_2.ConversationLoop
  nodes [Output Animation Pose(Result=Sequence Player 'ConversationLoop'() {sequence=&/ConversationLoop})]
 }
 ConversationFinish {
  path=FullBodyOverride.AnimGraphNode_StateMachine_0.NPC_Conversation.AnimStateNode_3.ConversationFinish
  nodes [Output Animation Pose(Result=Sequence Player 'ConversationFinish'() {bLoopAnimation=false, sequence=&/ConversationFinish})]
 }
 ConversationGiveItem {
  path=FullBodyOverride.AnimGraphNode_StateMachine_0.NPC_Conversation.AnimStateNode_4.ConversationGiveItem
  nodes [Output Animation Pose(Result=Sequence Player 'ConversationGiveItem'() {bLoopAnimation=false, sequence=&/ConversationGiveItem})]
 }
 ConversationReceiveItem {
  path=FullBodyOverride.AnimGraphNode_StateMachine_0.NPC_Conversation.AnimStateNode_5.ConversationReceiveItem
  nodes [Output Animation Pose(Result=Sequence Player 'ConversationReceiveItem'() {bLoopAnimation=false, sequence=&/ConversationReceiveItem})]
 }
 Transition {
  path=FullBodyOverride.AnimGraphNode_StateMachine_0.NPC_Conversation.AnimStateTransitionNode_0.Transition
  nodes [Result(bCanEnterTransition=Get IsConversationStarting() -> (IsConversationStarting:bool))]
 }
 Transition {path=FullBodyOverride.AnimGraphNode_StateMachine_0.NPC_Conversation.AnimStateTransitionNode_1.Transition, nodes [Result()]}
 Transition {
  path=FullBodyOverride.AnimGraphNode_StateMachine_0.NPC_Conversation.AnimStateTransitionNode_2.Transition
  nodes [Result(bCanEnterTransition=Get IsConversationEnding() -> (IsConversationEnding:bool))]
 }
 Transition {path=FullBodyOverride.AnimGraphNode_StateMachine_0.NPC_Conversation.AnimStateTransitionNode_3.Transition, nodes [Result()]}
 Transition {
  path=FullBodyOverride.AnimGraphNode_StateMachine_0.NPC_Conversation.AnimStateTransitionNode_4.Transition
  nodes [Result(bCanEnterTransition=Get IsGivingItem() -> (IsGivingItem:bool))]
 }
 Transition {
  path=FullBodyOverride.AnimGraphNode_StateMachine_0.NPC_Conversation.AnimStateTransitionNode_5.Transition
  nodes [Result(bCanEnterTransition=Get IsReceivingItem() -> (IsReceivingItem:bool))]
 }
 Transition {
  path=FullBodyOverride.AnimGraphNode_StateMachine_0.NPC_Conversation.AnimStateTransitionNode_6.Transition
  nodes [Result(bCanEnterTransition=Get IsConversationEnding())]
 }
 Transition {
  path=FullBodyOverride.AnimGraphNode_StateMachine_0.NPC_Conversation.AnimStateTransitionNode_7.Transition
  nodes [Result(bCanEnterTransition=Get IsConversationEnding())]
 }
 PreLegIK {nodes [
   Input Pose()
   Output Pose() {layerGroup=PlayerOverride}
  ]}
 UpperBody {nodes [
   Input Pose() {name=InPoseUpperBody}
   Output Pose() {layerGroup=PlayerOverride}
  ]}
]
variables [
 IsConversationEnding {type=boolean}
 IsConversationStarting {type=boolean}
 IsGivingItem {type=boolean}
 IsReceivingItem {type=boolean}
]
