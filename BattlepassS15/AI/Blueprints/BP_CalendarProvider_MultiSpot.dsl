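; Calendar-aware Blueprint for Season 15 battle pass AI that switches on Living World point providers.
; On BeginPlay it waits for playlist data, then enables the default provider if the named calendar event is inactive.
; A debug bypass or empty event name skips that default.
; As written, every valid calendar provider is then enabled regardless of the check, using the deprecated EnablePointProvider call.
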(event EventBeginPlay
  (bind _defaultpointprovider (Variables|Default|GetDefaultPointProvider))
  (bind _calendareventname (Variables|Default|GetCalendarEventName))
  (Class|FortAsyncActionCurrentPlaylistReady|PlaylistDataReadyAsync
    (:OnReady
      (if (Variables|Default|GetDebugBypassCalendarEvent)
        _
        (elif (Utilities|String|IsEmpty _calendareventname)
          _
          (else
            (bind _biseventactive (CalendarEvents|GetCalendarEventInformation _calendareventname))
            (if (not _biseventactive)
              (Utilities|IsValid _defaultpointprovider
                (:Is%20Valid
                  (Class|FortAthenaLivingWorldPointProviderInterface|EnablePointProvider_DEPRECATED _defaultpointprovider)))))))
      (for _array_element (Variables|Default|GetCalendarPointProviders)
        (Utilities|IsValid _array_element
          (:Is%20Valid
            (Class|FortAthenaLivingWorldPointProviderInterface|EnablePointProvider_DEPRECATED _array_element)))))))

(event Collision|EventActorBeginOverlap
  (params (OtherActor :type Actor%20Object%20Reference)))

(event EventTick
  (params (DeltaSeconds :type Float%20%28single-precision%29)))

(fn ConstructionScript)
